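# Sound cue for activating a timed trial quest in Athena, routed to the Fort_Quests sound class.
# A single wave player feeds the output directly with the TimedTrial_Warning wave, with a boosted volume multiplier.
# It overrides concurrency with a PreventNew rule, so new instances are refused once the limit is reached.
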SoundCue /Game/Athena/Sounds/Quests/TimedTrial/Athena_Quest_TimeTrial_Activate_Cue
root=WavePlayer_1
nodes [{id=WavePlayer_1, properties {soundWaveAssetPtr=&/TimedTrial_Warning}}]
properties {
 bOverrideConcurrency=true
 maxDistance=2097152
 soundClassObject=&Fort_Quests
 volumeMultiplier=2.24
 concurrencyOverrides {resolutionRule=PreventNew}
}
